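; Looping gameplay cue that keeps the in-water effect of a cage saddle vehicle matched to the water surface.
; On looping start it stores the first particle component, casts the instigator to CageSaddleVehicle and caches that vehicle's water body.
; A repeating 0.2-second timer, used instead of Tick, queries surface info at the vehicle's location and feeds plane location, normal, body index and depth into the effect's parameters.
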(event EventBeginPlay)

(event Collision|EventActorBeginOverlap
  (params (OtherActor :type Actor%20Object%20Reference)))

(event EventTick
  (params (DeltaSeconds :type Float%20%28single-precision%29)))

(event EventOnLoopingStart(EitherVFX-preferthisversion)
  (params (MyTarget :type Actor%20Object%20Reference) (Parameters :type Gameplay%20Cue%20Parameters%20Structure%20%28by%20ref%29) (ParticleComponents :type Array%20of%20FXSystem%20Component%20Object%20References) (AudioComponents :type Array%20of%20Audio%20Component%20Object%20References))
  (bind _result (ArrayMacros|SafeArrayGet ParticleComponents)
    (:%20%20
      (Variables|Default|SetReferenceToFX _result)
      (bind _ascage_saddle_vehicle (Utilities|Casting|CastToCageSaddleVehicle (Game|GetInstigator)))
      (Variables|Default|SetAs%20Cage%20Saddle%20Vehicle _ascage_saddle_vehicle)
      (Variables|Default|SetWater%20Body (Variables|/MotherGameplay/Items/CageSaddle/Example/CageSaddleVehicle.CageSaddleVehicle_C|GetWater%20Body _ascage_saddle_vehicle))
      (Utilities|Time|SetTimerbyEvent (delegate Custom|TickWaterInfo) 0.2 true))))

(event Custom|TickWaterInfo
  (bind _water_body (Variables|Default|GetWater%20Body))
  (bind _referencetofx (Variables|Default|GetReferenceToFX))
  (Utilities|IsValid _water_body
    (:Is%20Valid
      (bind _outwatersurfaceinfo (WaterBody|GetWaterSurfaceInfoStruct _water_body (Transformation|GetActorLocation (Variables|Default|GetAs%20Cage%20Saddle%20Vehicle)) true false false))
      (Effects|Components|ParticleSystem|SetVectorParameter _referencetofx "WaterPlaneLocation" (Utilities|Struct|BreakWaterSurfaceInfo _outwatersurfaceinfo))
      (Effects|Components|ParticleSystem|SetVectorParameter _referencetofx "WaterPlaneNormal" (.WaterPlaneNormal (Utilities|Struct|BreakWaterSurfaceInfo _outwatersurfaceinfo)))
      (Effects|Components|ParticleSystem|SetFloatParameter _referencetofx "WaterBodyIndex" (Math|Conversions|ToFloat(Integer) (.WaterBodyIdx (Utilities|Struct|BreakWaterSurfaceInfo _outwatersurfaceinfo))))
      (Effects|Components|ParticleSystem|SetFloatParameter _referencetofx "WaterDepth" (.WaterDepth (Utilities|Struct|BreakWaterSurfaceInfo _outwatersurfaceinfo))))))

(fn ConstructionScript)
