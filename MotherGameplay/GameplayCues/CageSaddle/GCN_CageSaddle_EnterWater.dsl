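; Burst gameplay cue notify that configures the water-entry effect for CageSaddle.
; On burst, it takes the first spawned FX component and casts the cue's source object to a Fort water body actor.
; It queries that body's surface info at the target actor's location.
; The resulting plane location, normal, body index (as float) and depth feed the effect's parameters, so the VFX can match the water surface.
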(fn OnBurst(EitherVFX-preferthisversion)
  (params (MyTarget :type Actor%20Object%20Reference) (Parameters :type Gameplay%20Cue%20Parameters%20Structure%20%28by%20ref%29) (ParticleComponents :type Array%20of%20FXSystem%20Component%20Object%20References) (AudioComponents :type Array%20of%20Audio%20Component%20Object%20References) (BurstCameraShakeInstance :type Camera%20Shake%20Base%20Object%20Reference) (BurstDecalInstance :type Decal%20Actor%20Object%20Reference))
  (bind (_normalizedmagnitude _rawmagnitude _effectcontext _matchedtagname _originaltag _aggregatedsourcetags _aggregatedtargettags _location _normal _instigator _effectcauser _sourceobject) (Ability|GameplayCue|BreakGameplayCueParameters Parameters))
  (bind _fx (LocalVariables|GetFX))
  (LocalVariables|SetFX (Utilities|Array|Get(acopy) ParticleComponents))
  (bind _asfort_water_body_actor (Utilities|Casting|CastToFortWaterBodyActor _sourceobject))
  (LocalVariables|SetWaterBody _asfort_water_body_actor)
  (bind _outwatersurfaceinfo (WaterBody|GetWaterSurfaceInfoStruct (LocalVariables|GetWaterBody) (Transformation|GetActorLocation MyTarget) true false false))
  (Effects|Components|ParticleSystem|SetVectorParameter _fx "WaterPlaneLocation" (Utilities|Struct|BreakWaterSurfaceInfo _outwatersurfaceinfo))
  (Effects|Components|ParticleSystem|SetVectorParameter _fx "WaterPlaneNormal" (.WaterPlaneNormal (Utilities|Struct|BreakWaterSurfaceInfo _outwatersurfaceinfo)))
  (Effects|Components|ParticleSystem|SetFloatParameter _fx "WaterBodyIndex" (Math|Conversions|ToFloat(Integer) (.WaterBodyIdx (Utilities|Struct|BreakWaterSurfaceInfo _outwatersurfaceinfo))))
  (Effects|Components|ParticleSystem|SetFloatParameter _fx "WaterDepth" (.WaterDepth (Utilities|Struct|BreakWaterSurfaceInfo _outwatersurfaceinfo))))
